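# Translucent UI material for the Ice Gauntlets reticle, layering a mirrored gauntlet shape, drop lines, a middle ring and a radial progress arc.
# MSDF textures and circle functions form anti-aliased shapes and shadows, with parameters for ring size, shape spacing, glow and progress.
# Panning ice-crack and noise textures drive a pulsing three-color blue fill, which CooldownColor blends against the shared UI white.
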Material /TeaCakeIceGauntlets/Art/Materials/M_UI_IceGauntlets_Reticle
graph {
 properties {
  bAutomaticallySetUsageInEditor=false
  bUsedWithStaticMesh=false
  blendMode=BLEND_Translucent
  materialDomain=MD_UI
  usedShadingModels=Default Lit
 }
 nodes [
  n1 = LinearInterpolate(A=NamedRerouteUsage(), B=Multiply(A=NamedRerouteUsage(), B=NamedRerouteUsage()), Alpha=NamedRerouteUsage())
  n2 = Multiply(A=CollectionParameter() {collection=/BattleRoyaleGameUI/Shared/MPC_BRGameUI, parameterName=White}, B=NamedRerouteUsage())
  Add(
   A=Add(A=Add(A=@n1, B=@n2), B=CollectionParameter() {collection=&MPC_BRGameUI, parameterName=Charcoal})
   B=Multiply(A=CollectionParameter() {collection=&MPC_BRGameUI, parameterName=White}, B=NamedRerouteUsage())
  )
  n3 = NamedRerouteUsage()
  n4 = Multiply(A=@n3) {constB=0.5}
  n5 = Multiply(A=NamedRerouteUsage(), B=@n3)
  n6 = NamedRerouteUsage()
  n7 = Constant() {r=0.85}
  n8 = Add(A=Add(A=LinearInterpolate(A=@n4, B=@n5) {constAlpha=0.6}, B=Multiply(A=@n6, B=@n7)), B=Multiply(A=NamedRerouteUsage(), B=Constant() {r=0.2}))
  n9 = Multiply(A=NamedRerouteUsage()) {constB=0.65}
  n10 = NamedRerouteUsage()
  Add(A=Add(A=Add(A=@n8, B=@n9), B=Multiply(A=@n10) {constB=0.9}), B=Multiply(A=NamedRerouteUsage()) {constB=0.3})
  LinearInterpolate(Alpha=ScalarParameter() {parameterName=CooldownReady}) {constA=0.6, constB=0.8}
  LinearInterpolate(Alpha=ScalarParameter() {parameterName=CooldownReady}) {constB=0.6}
  n11 = Abs(None=ConstantBiasScale(None=TextureCoordinate()) {bias=-0.1, scale=2})
  n12 = Time()
  n13 = TextureSample(UVs=Add(A=Multiply(A=@n11) {constB=2}, B=Panner(Time=@n12) {speedX=0.1, speedY=0.1})) {samplerType=SAMPLERTYPE_Masks, texture=/Game/Environments/Asteria/Terrain/Textures/T_Asteria_Ice_Cracks_M}
  n14 = Time()
  n15 = TextureSample(UVs=Add(B=Panner(Time=@n14) {speedX=-0.03})) {texture=/Game/Global/Textures/Default/Noise/T_UI_Ice_Noise}
  n16 = Add(A=Multiply(A=OneMinus(None=@n13.R)) {constB=20}, B=Multiply(A=@n15.R) {constB=15}) {constB=0}
  n17 = Multiply(A=TextureSample(UVs=Panner(Time=@n16) {speedY=0.02}) {texture=&T_UI_Ice_Noise}.R, B=@n15.G)
  n18 = Multiply(A=Time())
  n19 = LinearInterpolate(Alpha=Saturate(None=Sine(None=@n18) {period=1.5})) {constA=0.4, constAlpha=1, constB=0.5}
  n20 = AppendVector(
   A=MaterialFunctionCall(
    A=Constant4Vector() {constant=<rgba 0 0.1445 0.5781 1>}.RGBA
    B=Constant4Vector() {constant=<rgba 0 0.7675 1 1>}.RGBA
    C=Constant4Vector() {constant=<rgba 0.2969 0.837 1 1>}.RGBA
    Alpha=Power(Base=@n17, Exp=@n19) {constExponent=0.5}
   ) {materialFunction=&3ColorBlend}.Result
   B=Constant()
  )
  n21 = ScalarParameter() {defaultValue=1.3, parameterName=Ice Power}
  NamedRerouteDeclaration(None=LinearInterpolate(
    A=CollectionParameter() {collection=&MPC_BRGameUI, parameterName=White}
    B=Multiply(A=@n20, B=@n21)
    Alpha=ScalarParameter() {parameterName=CooldownColor}
   ) {constA=1}) {name=FillColor, nodeColor=<rgba 0 0 0 1>}
  n22 = TextureCoordinate()
  n23 = AppendVector(A=Constant(), B=Constant() {r=0.18})
  n24 = Constant() {r=0.14}
  n25 = Constant() {r=0.25}
  n26 = ScalarParameter() {parameterName=Ring Size}
  NamedRerouteDeclaration(None=MaterialFunctionCall(UVs=Add(A=@n22, B=@n23), Size=LinearInterpolate(A=@n24, B=@n25, Alpha=@n26), OutlineThickness=Constant() {r=0.018}) {materialFunction=&MF_UI_Circle}.Outline) {name=MiddleCircle, nodeColor=<rgba 0 0 0 1>}
  n27 = TextureCoordinate()
  n28 = AppendVector(A=Constant(), B=Constant() {r=-0.4})
  n29 = Constant() {r=0.3}
  n30 = Constant() {r=0.3}
  n31 = TextureSample(UVs=MaterialFunctionCall(UVs=Add(A=@n27, B=@n28), Texture Scale=AppendVector(A=@n29, B=@n30)) {materialFunction=&ScaleUVsByCenter}.UVs) {
   samplerSource=SSM_Clamp_WorldGroupSettings
   samplerType=SAMPLERTYPE_LinearColor
   texture=/TeaCakeIceGauntlets/Art/Textures/T_UI_IceKingGauntlets_DropLines
  }
  n32 = MaterialFunctionCall!(MSDF RGB=@n31.RGB) {materialFunction=/FortUILibrary/MaterialFunctions/Shapes/Modifiers/MF_UI_MSDFtoSDF}
  n33 = MaterialFunctionCall(DistanceField=@n32.SDF, Threshold=Constant() {r=0.35}) {materialFunction=&MF_UI_AntiAliasing}
  n34 = MaterialFunctionCall(DistanceField=@n32.SDF, Threshold=Constant() {r=0.6}) {materialFunction=&MF_UI_AntiAliasing}
  NamedRerouteDeclaration(None=Subtract(
    A=NamedRerouteDeclaration(None=@n33.Outer) {name=DropLinesShapeAndShadow, nodeColor=<rgba 0 0 0 1>}
    B=NamedRerouteDeclaration(None=@n34.Outer) {name=DropLinesShape, nodeColor=<rgba 0 0 0 1>}
   )) {name=DropLinesShadow, nodeColor=<rgba 0 0 0 1>}
  n35 = TextureCoordinate()
  n36 = AppendVector(A=Constant(), B=Constant() {r=0.18})
  n37 = Constant() {r=0.14}
  n38 = Constant() {r=0.25}
  n39 = ScalarParameter() {parameterName=Ring Size}
  NamedRerouteDeclaration(None=MaterialFunctionCall(UVs=Add(A=@n35, B=@n36), Size=LinearInterpolate(A=@n37, B=@n38, Alpha=@n39), OutlineThickness=Constant() {r=0.04}) {materialFunction=&MF_UI_Circle}.Outline) {name=MiddleCircleShadow, nodeColor=<rgba 0 0 0 1>}
  n40 = MaterialFunctionCall!(UV=TextureCoordinate(), Mirror Y, Swizzle UV) {materialFunction=/FortUILibrary/MaterialFunctions/Basics/MF_UI_Symmetry_2x1_New}
  n41 = AppendVector(A=ScalarParameter() {parameterName=Shape Distance}, B=Constant() {r=0.15})
  n42 = Constant() {r=1.5}
  n43 = Constant() {r=0.7}
  n44 = TextureSample(UVs=MaterialFunctionCall(UVs=Add(A=@n40.Symmetry, B=@n41), Texture Scale=AppendVector(A=@n42, B=@n43)) {materialFunction=&ScaleUVsByCenter}.UVs) {
   samplerSource=SSM_Clamp_WorldGroupSettings
   samplerType=SAMPLERTYPE_LinearColor
   texture=/TeaCakeIceGauntlets/Art/Textures/T_UI_IceKingGauntlets_Shape
  }
  n45 = MaterialFunctionCall(MSDF RGB=@n44.RGB) {materialFunction=&MF_UI_MSDFtoSDF}
  n46 = MaterialFunctionCall(DistanceField=@n45.SDF, Threshold=Constant() {r=0.3}) {materialFunction=&MF_UI_AntiAliasing}
  n47 = ScalarParameter() {parameterName=ShapeGlow}
  n48 = NamedRerouteDeclaration(None=MaterialFunctionCall(DistanceField=@n45.SDF, Threshold=Constant() {r=0.7}, Smoothing=LinearInterpolate(Alpha=@n47) {constB=0.9}) {materialFunction=&MF_UI_AntiAliasing}.Outer) {name=Shape, nodeColor=<rgba 0 0 0 1>}
  n49 = Constant() {r=0.31}
  NamedRerouteDeclaration(None=Subtract(
    A=NamedRerouteDeclaration(None=@n46.Outer) {name=Shape+Shadow, nodeColor=<rgba 0 0 0 1>}
    B=LinearInterpolate(A=@n46.Outer, B=@n48, Alpha=@n49)
   )) {name=ShapeShadowOnly, nodeColor=<rgba 0 0 0 1>}
  n50 = Add(A=TextureCoordinate(), B=AppendVector(A=Constant(), B=Constant() {r=0.2}))
  n51 = OneMinus(None=LinearInterpolate(Alpha=ScalarParameter() {defaultValue=1, parameterName=Progress}) {constA=0.12, constB=0.82})
  NamedRerouteDeclaration(None=SmoothStep(Value=MaterialFunctionCall!(UV=@n50, "Arc Start Angle (0-1)", "Arc End Angle (0-1)", "Fill Ratio (0-1)"=@n51, Fill Clockwise) {materialFunction=/FortUILibrary/MaterialFunctions/Basics/MF_UI_RadialArc}.Centered Fill) {constMax=0.001}) {name=ProgressBar, nodeColor=<rgba 0 0 0 1>}
 ]
}
