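# Reticle widget for the Ice Gauntlets weapon, showing projectile cooldown and playing intro, attack and cooldown-end animations on one material-driven image.
# View-target changes cache the pawn, controller and item, and rebind ammo events that trigger the attack animation.
# Async gameplay-tag waits toggle cooldown; a looping timer feeds remaining time into material scalars.
# Squared-distance checks on the reticle actor toggle an in-range animation, which has no tracks.
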WidgetBlueprint /TeaCakeIceGauntlets/Widgets/WBP_IceGauntlets_Reticle
animations [
 Anim_TargetInRange {
  binding_count=0
  length_seconds=0.25
  object=&self:Anim_TargetInRange
  sequence {bindings []}
 }
 Anim_Cooldown_Active_New {
  binding_count=1
  length_seconds=15
  object=&self:Anim_Cooldown_Active_New
  sequence {bindings [Reticle {
     id=#1
     possessed_class=Image
     tracks [{
       type=MovieSceneWidgetMaterialTrack
       sections [
        {type=MovieSceneParameterSection, end_frame=299, start_frame=0}
        {type=MovieSceneParameterSection, end_frame=299, start_frame=0}
        {
         type=MovieSceneParameterSection
         end_frame=300
         start_frame=0
         channels [Progress {type=MovieSceneScriptingFloatChannel, keys [[0,0],[299.98,1]]}]
        }
       ]
      }]
    }]}
 }
 Anim_Cooldown_End_New {
  binding_count=1
  length_seconds=0.8334
  object=&self:Anim_Cooldown_End_New
  sequence {
   bindings [Reticle {
     id=#2
     possessed_class=Image
     tracks [{
       type=MovieSceneWidgetMaterialTrack
       sections [{
         type=MovieSceneParameterSection
         end_frame=50
         start_frame=0
         channels [
          CooldownColor {type=MovieSceneScriptingFloatChannel, keys [[0,1],[35,1],[48,0]]}
          ShapeGlow {type=MovieSceneScriptingFloatChannel, keys [[0,0],[25,1],[48,0]]}
          Ice Power {type=MovieSceneScriptingFloatChannel, keys [[8,1.3],[31,4],[35,2],[47,2],[48,1.3]]}
         ]
        }]
      }]
    }]
   display_rate {numerator=60}
  }
 }
 Anim_Attack_New {
  binding_count=1
  length_seconds=0.3
  object=&self:Anim_Attack_New
  sequence {bindings [Reticle {
     id=#3
     possessed_class=Image
     tracks [{
       type=MovieSceneWidgetMaterialTrack
       sections [{
         type=MovieSceneParameterSection
         end_frame=6
         start_frame=0
         channels [Shape Distance {type=MovieSceneScriptingFloatChannel, keys [[0,0],[3,0.06],[6,0]]}]
        }]
      }]
    }]}
 }
 Anim_Intro {
  binding_count=1
  length_seconds=0.4
  object=&self:Anim_Intro
  sequence {bindings [Reticle {
     id=#4
     possessed_class=Image
     tracks [
      {
       type=MovieSceneWidgetMaterialTrack
       sections [{
         type=MovieSceneParameterSection
         end_frame=8
         start_frame=0
         channels [
          Shape Distance {type=MovieSceneScriptingFloatChannel, keys [[0,0.1],[8,0]]}
          Ring Size {type=MovieSceneScriptingFloatChannel, keys [[2,1],[8,0]]}
         ]
        }]
      }
      {
       type=MovieSceneColorTrack
       display_name=Color and Opacity
       sections [{
         type=MovieSceneColorSection
         end_frame=8
         start_frame=0
         channels [
          Color.R {type=MovieSceneScriptingFloatChannel, const [1,1]}
          Color.G {type=MovieSceneScriptingFloatChannel, const [1,1]}
          Color.B {type=MovieSceneScriptingFloatChannel, const [1,1]}
          Color.A {type=MovieSceneScriptingFloatChannel, keys [[0,0],[8,1]]}
         ]
        }]
      }
     ]
    }]}
 }
]
graphs [
 BindWeaponComponentEvents {nodes [
   BindWeaponComponentEvents!() -> (then, Weapon Component:Fort Weapon Object Reference) {is=title K2Node_FunctionEntry}
   > n1 = Is Valid(InputObject=Get B Weap Ice Gauntlets Component() -> (B Weap Ice Gauntlets Component:B Weap Ice Gauntlets Object Reference))
   > Unbind Event from On Local Ammo Changed(self:Fort Weapon Object Reference=Get B Weap Ice Gauntlets Component(), Delegate=Create Event()) {is=title K2Node_RemoveDelegate}
   > n2 = Set B Weap Ice Gauntlets Component(B Weap Ice Gauntlets Component:B Weap Ice Gauntlets Object Reference) -> (Output_Get:B Weap Ice Gauntlets Object Reference)
   n3 = Cast To B_Weap_IceGauntlets!(execute=[@n2.then, @n1.Is Not Valid], Object=Get Weapon Component() -> (Weapon Component:Fort Weapon Object Reference)) -> (then, CastFailed, AsB Weap Ice Gauntlets:B Weap Ice Gauntlets Object Reference) {is=title K2Node_DynamicCast}
   > Is Valid(InputObject=@n3)
   > n4 = Set B Weap Ice Gauntlets Component(B Weap Ice Gauntlets Component=@n3)
   > Bind Event to On Local Ammo Changed(self:Fort Weapon Object Reference=@n4, Delegate=Create Event()) {is=title K2Node_AddDelegate}
  ]}
 EventGraph {nodes [
   Bind Event to On Actor Under Reticle Set BP(self:Fort Player Controller Object Reference, Delegate=Create Event()) {is=title K2Node_AddDelegate}
   > OnActorUnderReticleSet!(
    execute
    self
    ActorUnderReticle:Actor Object Reference=GetActorUnderReticle!(self:Fort Player Controller Object Reference) -> (ReturnValue:Actor Object Reference)
   ) -> (then)
   OnLocalAmmoChanged!() -> (OutputDelegate, then, LocalCount:int, LocalRemaining:int) {is=title K2Node_CustomEvent}
   > PlayAnimation(InAnimation=Get Anim_Attack_New() -> (Anim_Attack_New:Widget Animation Object Reference))
   n1 = OnViewTargetChanged!() -> (
    OutputDelegate
    then
    PlayerController:Fort Player Controller Object Reference
    OldViewTarget:Actor Object Reference
    NewViewTarget:Actor Object Reference
   ) {is=title K2Node_CustomEvent}
   > n2 = Cast To PlayerPawn_Athena(Object=@n1.NewViewTarget)
   > Set ViewTargetPlayerPawn(ViewTargetPlayerPawn:Player Pawn Athena Object Reference=@n2) -> (Output_Get:Player Pawn Athena Object Reference)
   > n3 = Sequence!(execute) -> (then_0, then_1, then_2:Exec)
   n4 = GetEquippedWeapon!(self:Fort Weapon Reticle Extension Widget Base Object Reference) -> (ReturnValue:Fort Weapon Object Reference)
   > Is Valid(InputObject=@n4)
   > BindWeaponComponentEvents!(execute, self, Weapon Component:Fort Weapon Object Reference=@n4) -> (then) {is=title K2Node_CallFunction}
   > Set CachedFortItem(CachedFortItem:Fort Item Object Reference=GetInventoryItem!(self:Fort Weapon Object Reference=@n4) -> (ReturnValue:Fort Item Object Reference)) -> (Output_Get:Fort Item Object Reference)
   > n5 = Cast To FortPlayerController(Object=GetController!(self:Pawn Object Reference=Get ViewTargetPlayerPawn() -> (ViewTargetPlayerPawn:Player Pawn Athena Object Reference)) -> (ReturnValue:Controller Object Reference))
   > Set ViewTarget_FortPlayerController(ViewTarget_FortPlayerController:Fort Player Controller Object Reference=@n5) -> (Output_Get:Fort Player Controller Object Reference)
   > UpdateProjectileCooldown()
   > PlayAnimation(InAnimation=Get Anim_Intro() -> (Anim_Intro:Widget Animation Object Reference))
   n6 = WaitGameplayTagAddToActor!(
    execute:Exec=@n3.then_1
    TargetActor:Actor Object Reference=Get ViewTargetPlayerPawn()
    Tag:Gameplay Tag Structure=Get ProjectileCooldownTag() -> (ProjectileCooldownTag:Gameplay Tag Structure)
    OnlyTriggerOnce:bool=false
   ) -> (then:Exec, AsyncTaskProxy:Ability Async Wait Gameplay Tag Added Object Reference, Added:Exec) {is=title K2Node_AsyncAction}
   Set InCooldown(execute=@n6.Added, InCooldown:bool=true) -> (Output_Get:bool)
   > Set TargetInRange(TargetInRange:bool=false) -> (Output_Get:bool)
   > StartCooldownProgress()
   n7 = WaitGameplayTagRemoveFromActor!(
    execute:Exec=@n3.then_2
    TargetActor:Actor Object Reference=Get ViewTargetPlayerPawn()
    Tag:Gameplay Tag Structure=Get ProjectileCooldownTag()
    OnlyTriggerOnce:bool=false
   ) -> (then:Exec, AsyncTaskProxy:Ability Async Wait Gameplay Tag Removed Object Reference, Removed:Exec) {is=title K2Node_AsyncAction}
   n8 = Branch(execute=@n7.Removed, Condition=IsAnimationPlaying(InAnimation=Get Anim_Intro()))
   Set InCooldown(execute=@n8.else, InCooldown=false)
   > SetScalarParameterValue(self=Get ReticleDynamicMaterial() -> (ReticleDynamicMaterial:Material Instance Dynamic Object Reference), ParameterName=Progress, Value=1)
   > Clear and Invalidate Timer by Handle(Handle=Get Timer_UpdateProjectileCooldown() -> (Timer_UpdateProjectileCooldown:Timer Handle Structure))
   > PlaySound2D(Sound=/TeaCakeIceGauntlets/Assets/Audio/MSS_TeaCake_IceGauntlets_Secondary_UI_MeterFill)
   > PlayAnimation(self=Self-Reference(), InAnimation=Get Anim_Cooldown_End_New() -> (Anim_Cooldown_End_New:Widget Animation Object Reference))
   StartCooldownProgress() {is=title K2Node_CustomEvent}
   > SetScalarParameterValue(self=Get ReticleDynamicMaterial(), ParameterName=Progress, Value=0)
   > n9 = Set Timer by Event(Delegate=Create Event(), Time=0.1, bLooping=true, bMaxOncePerFrame=true)
   > Set Timer_UpdateProjectileCooldown(Timer_UpdateProjectileCooldown:Timer Handle Structure=@n9) -> (Output_Get:Timer Handle Structure)
   Event Construct()
   n10 = GetOwningPlayer()
   > n11 = Cast To FortPlayerController(Object=@n10)
   n12 = Create Event()
   > Bind Event to On View Target Changed(self:Fort Player Controller Object Reference=@n11, Delegate=@n12) {is=title K2Node_AddDelegate}
   > OnViewTargetChanged!(
    execute
    self
    PlayerController:Fort Player Controller Object Reference
    OldViewTarget:Actor Object Reference
    NewViewTarget:Actor Object Reference=GetViewTarget!(self:Controller Object Reference=@n10) -> (ReturnValue:Actor Object Reference)
   ) -> (then) {is=title K2Node_CallFunction}
   > n13 = GetDynamicMaterial(self=Get Reticle() -> (Reticle:Image Object Reference))
   > Set ReticleDynamicMaterial(ReticleDynamicMaterial:Material Instance Dynamic Object Reference=@n13) -> (Output_Get:Material Instance Dynamic Object Reference)
   Event Destruct()
   > n14 = Cast To FortPlayerController(Object=@n10)
   > Unbind Event from On View Target Changed(self:Fort Player Controller Object Reference=@n14, Delegate=@n12) {is=title K2Node_RemoveDelegate}
   > BindWeaponComponentEvents()
   > Branch(Condition=Is Valid Timer Handle!(Handle:Timer Handle Structure=Get Timer_UpdateProjectileCooldown()) -> (ReturnValue:bool))
   > Clear and Invalidate Timer by Handle(Handle=Get Timer_UpdateProjectileCooldown())
   OnActorUnderReticleSet(ActorUnderReticle=GetActorUnderReticle(self=Get ViewTarget_FortPlayerController() -> (ViewTarget_FortPlayerController:Fort Player Controller Object Reference)))
   Reroute Node(InputPin:Wildcard) -> (OutputPin:Wildcard)
  ]}
 OnActorUnderReticleSet {nodes [
   n1 = OnActorUnderReticleSet!() -> (then, ActorUnderReticle:Actor Object Reference) {is=title K2Node_FunctionEntry}
   > n2 = Is Valid(InputObject=@n1)
   > n3 = Branch(Condition=Get InCooldown() -> (InCooldown:bool))
   n4 = Get Actor Location(self=Get ViewTargetPlayerPawn())
   n5 = Get Actor Location(self=@n1)
   n6 = GetValueAtLevel!(Input:"Scalable Float Structure (by ref)"=Get SF_MaxProjectileRange() -> (SF_MaxProjectileRange:Scalable Float Structure), Level:float=0) -> (ReturnValue:float)
   n7 = Branch(
    execute=@n3.else
    Condition="float < float"(A="Distance Squared (Vector)"!(V1:Vector=@n5, V2:Vector=@n4) -> (ReturnValue:double), B=Square!(A:double=@n6) -> (ReturnValue:double))
   )
   > n8 = Branch(Condition=Get TargetInRange() -> (TargetInRange:bool))
   PlayAnimation(execute=@n8.else, InAnimation=Get Anim_TargetInRange() -> (Anim_TargetInRange:Widget Animation Object Reference), PlaybackSpeed=4)
   > Set TargetInRange(TargetInRange=true)
   Branch(execute=[@n7.else, @n2.Is Not Valid], Condition=Get TargetInRange())
   > PlayAnimation(InAnimation=Get Anim_TargetInRange(), PlayMode=Reverse, PlaybackSpeed=4)
   > Set TargetInRange(TargetInRange=false)
  ]}
 UpdateProjectileCooldown {nodes [
   UpdateProjectileCooldown() {is=title K2Node_FunctionEntry}
   n1 = Get ViewTarget_FortPlayerController()
   > n2 = Is Valid(InputObject=@n1)
   Clear and Invalidate Timer by Handle(execute=@n2.Is Not Valid, Handle=Get Timer_UpdateProjectileCooldown())
   n3 = GetCoolDownTimeRemaining!(
    self:Fort Item Object Reference=Get CachedFortItem() -> (CachedFortItem:Fort Item Object Reference)
    FortPC:Fort Player Controller Object Reference=@n1
   ) -> (CooldownDuration:float, ReturnValue:float)
   n4 = SafeDivide(A=float - float(A=@n3, B=@n3.ReturnValue), B=@n3)
   n5 = "Nearly Equal (Float)"(A=@n4, B=0, ErrorTolerance=0)
   SetScalarParameterValue(execute=@n2.Is Valid, self=Get ReticleDynamicMaterial(), ParameterName=Progress, Value=SelectFloat(A=1, B=@n4, bPickA=@n5))
   > SetScalarParameterValue(self=Get ReticleDynamicMaterial(), ParameterName=CooldownColor, Value=SelectFloat(A=0, B=1, bPickA=@n5))
  ]}
]
variables [
 B Weap Ice Gauntlets Component {type=B_Weap_IceGauntlets_C, default=null}
 CachedFortItem {type=FortItem, default=null}
 InCooldown {type=boolean}
 ProjectileCooldownTag {type=GameplayTag, default {tagName=Granted.Cooldown.IceGauntlets.Secondary}}
 ReticleDynamicMaterial {type=MaterialInstanceDynamic}
 SF_MaxProjectileRange {type=ScalableFloat, default {registryType=None, value=0, curve {curveTable=null, rowName=LemonExam.Projectile.MaxRange}}}
 TargetInRange {type=boolean}
 Timer_UpdateProjectileCooldown {type=TimerHandle}
 ViewTarget_FortPlayerController {type=FortPlayerController, default=null}
 ViewTargetPlayerPawn {type=PlayerPawn_Athena_C}
]
widgets {
 info {
  inheritedWidgetCount=1
  parentClass=/Script/FortniteUI.FortWeaponReticleExtensionWidgetBase
  rootWidgetClass=&Overlay
  widgetCount=3
 }
 widgets [
  Overlay_Root {widgetClassPath=&Overlay, widget=Overlay_Root}
  Reticle {
   widgetClassPath=&Image
   bIsVariable=true
   parent=Overlay_Root
   widget=Reticle
   properties {
    visibility=SelfHitTestInvisible
    brush {
     bIsDynamicallyLoaded=false
     drawAs=Image
     imageType=FullColor
     mirroring=NoMirror
     resourceName=None
     resourceObject=/TeaCakeIceGauntlets/Art/Materials/M_UI_IceGauntlets_Reticle
     tiling=NoTile
     imageSize=<128 128>
     margin=<ltrb 0 0 0 0>
     outlineSettings {
      bUseBrushTransparency=false
      roundingType=HalfHeightRadius
      width=0
      color {colorUseRule=UseColor_Specified, specifiedColor=<rgba 0 0 0 0>}
      cornerRadii=<0 0 0 0>
     }
     tintColor {colorUseRule=UseColor_Specified, specifiedColor=<rgba 1 1 1 1>}
     uVRegion {bIsValid=false, max=<0 0>, min=<0 0>}
    }
   }
   slot {class=OverlaySlot, properties {horizontalAlignment=HAlign_Center, verticalAlignment=VAlign_Center, padding=<ltrb 0 22 0 0>}}
  }
  TutorialHighlight {widgetClassPath=&TutorialHighlight}
 ]
}
